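# Niagara effect for the DimeBlanketGrace back bling, emitting hearts and flickering sparkle sprites on the backpack.
# Emitters shown follow System.DimeBlanketSkeletal, using baked offset arrays or selected socket indices to shape placement.
# User.Variant blends pink and blue palettes on hearts and sparkles.
# Spawning and brightness respond to music-reactive parameters, front-end versus gameplay, time of day, platform or distance, with a cosine front-end pulse when the music preview value misses its comparison.
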NiagaraSystem /BRCosmetics/Accessories/FORT_Backpacks/Backpack_F_MED_DimeBlanketGrace/FX/NS_Backpack_DimeBlanketGrace
dependencies {
 usedDataInterfaces [
  &NiagaraDataInterfaceCurve
  &NiagaraDataInterfaceColorCurve
  &NiagaraDataInterfacePlatformSet
  &NiagaraDataInterfaceArrayFloat3
  &NiagaraDataInterfaceArrayInt32
  &NiagaraDataInterfaceSkeletalMesh
 ]
 usedDynamicInputs [&Multiply_Vector2D_ByFloat, &MakeVector2D]
 usedModules [
  &self:SystemSpawnScript.SystemScriptSource.SystemScriptGraph.NiagaraNodeAssignment_2.SetVariables_#1
  &self:SystemSpawnScript.SystemScriptSource.SystemScriptGraph.NiagaraNodeAssignment_3.SetVariables_#2
  &SystemState
  &self:SystemSpawnScript.SystemScriptSource.SystemScriptGraph.NiagaraNodeAssignment_6.SetVariables_#3
  &self:SystemSpawnScript.SystemScriptSource.SystemScriptGraph.NiagaraNodeAssignment_7.SetVariables_#4
  &self:SystemSpawnScript.SystemScriptSource.SystemScriptGraph.NiagaraNodeAssignment_0.SetVariables_#5
  /Niagara/Noise/Waveform
 ]
}
emitters {
 Hearts001 {
  emitterUpdateScript [
   &EmitterState()
   &SpawnRate(
    Spawn Probability=&FN_FrontEndFloatModifier(
     Gameplay=&MakeCustomFloatFromBool(Bool=&SetBoolByFloatComparison(A=$System.Music_Reactive_FX_Preview, B=0), True Float=1, False Float=0)
     Front End=&MakeCustomFloatFromBool(Bool=&SetBoolByFloatComparison(A=$System.Music_Reactive_FX_Preview, B=0), True Float=1, False Float=0.75)
    )
    SpawnRate=&FN_FrontEndFloatModifier(
     Gameplay=&Multiply_Float(
      A=10
      B=&MakeCustomFloatFromBool(Bool=$User.MusicReactive, True Float=1, False Float=$System.DistanceSpawn)
     )
     Front End=&Multiply_Float(A=10, B=&MakeCustomFloatFromBool(Bool=$User.MusicReactive, True Float=1, False Float=1))
    )
    Use Spawn Probability=true
   )
  ]
  particleSpawnScript [
   &InitializeParticle(
    AdjustAlpha=true
    AdjustHue=true
    AdjustSaturation=true
    AdjustValue=true
    Color=&LerpLinearColors(StartColor=<rgba 1 0.5795 0.9631 0.4>, EndColor=<rgba 0.7758 0.8615 1 0.5>, LerpFactor=$User.Variant)
    Color Mode=::NewEnumerator3
    Lifetime Max=0.6
    Lifetime Min=0.4
    Lifetime Mode=::NewEnumerator1
    Sprite Size Mode=::NewEnumerator1
    Sprite UV Mode=::NewEnumerator1
    Uniform Sprite Size Max=8
    Uniform Sprite Size Min=4
    UsePositionOffset=false
    Write Material Random=true
   )
   &CameraOffset(Camera Offset Amount=&FN_FrontEndFloatModifier(Gameplay=12, Front End=2))
  ]
  particleUpdateScript [
   &ParticleState()
   &ScaleSpriteSize(Uniform Curve Sprite Scale=DataInterface() {propertyValues {Curve {keys [
        {arriveTangent=2.4668, interpMode=RCIM_Cubic, leaveTangent=2.4668, tangentMode=RCTM_User, time=0, value=0}
        {tangentMode=RCTM_User, time=1, value=1}
       ]}}})
   &ScaleColor(
    Scale Alpha=&Multiply_Float(
     A=&Multiply_Float(
      A=&FloatFromCurve(
       FloatCurve=DataInterface() {propertyValues {Curve {keys [<at 0 1 RCIM_Cubic>,<at 1 0>]}}}
       Scale Curve=&RandomRangeFloat(Minimum=1, Maximum=1)
      )
      B=&MakeFloatFromLinearColor(LinearColor=$System.TODColorMultiplier, Channel:ENiagaraLinearColor_Channels=::NewEnumerator3)
     )
     B=&MakeCustomFloatFromBool(
      Bool=$User.MusicReactive
      True Float=1
      False Float=&MakeCustomFloatFromBool(
       Bool=&SetBoolByFloatComparison(A=$System.Music_Reactive_FX_Preview, B=0)
       True Float=&Multiply_Float(
        A=&Lerp_Float(
         A=&FN_FrontEndFloatModifier(
          Gameplay=&FN_PickFloatDependOnPlatform(TargetPlatformActive=1, TargetPlatformNoActive=0.8)
          Front End=0.6
         )
         B=&FN_FrontEndFloatModifier(
          Gameplay=&FN_PickFloatDependOnPlatform(TargetPlatformActive=3, TargetPlatformNoActive=1.5)
          Front End=1
         )
         Alpha=$System.Music_Reactive_FX_Preview
        )
        B=&FNTimeOfDayScalar(Night Scale=0.4, Morning Scale=0.75, Day Scale=0.75, Evening Scale=0.75)
       )
       False Float=&FN_FrontEndFloatModifier(
        Gameplay=0
        Front End=&FloatFromCurve(
         FloatCurve=DataInterface() {propertyValues {Curve {keys [<at 0 0 RCIM_Cubic>, {arriveTangent=1.362, interpMode=RCIM_Cubic, leaveTangent=-0.9886, tangentMode=RCTM_Break, time=0.5, value=1}, <at 1 0>]}}}
         CurveIndex=&Cosine(
          Normalized Angle=&Add_Float(A=&Divide_Float(A=$Engine.Time, B=5), B=0.5)
          Period=1
          Scale=0.5
          Bias=1
         )
        )
       )
      )
     )
    )
    Scale RGB=&Multiply_Vector(
     A=&Lerp_Vector(
      A=&MakeVectorFromLinearColorRGB(Color=&MergeHelper_ColorSwitch(
        0=&ColorFromCurve(ColorCurve=DataInterface() {propertyValues {
           AlphaCurve {keys [<at 0 1 RCIM_Cubic>,<at 1 0 RCIM_Cubic>]}
           BlueCurve {keys [<at 0 6.1 RCIM_Cubic>,<at 0.2254 3.6>,<at 0.4718 1.6667>,<at 0.8284 0.2104 RCIM_Cubic>]}
           GreenCurve {keys [<at 0 6 RCIM_Cubic>,<at 0.2254 3.2>,<at 0.4718 1.2>,<at 0.8284 0.1052 RCIM_Cubic>]}
           RedCurve {keys [<at 0 12 RCIM_Cubic>,<at 0.2254 8>,<at 0.4718 4>,<at 0.8284 0.526 RCIM_Cubic>]}
          }})
        1=&ColorFromCurve(ColorCurve=DataInterface() {propertyValues {
           AlphaCurve {keys [<at 0 1 RCIM_Cubic>,<at 1 0 RCIM_Cubic>]}
           BlueCurve {keys [<at 0 9.6056 RCIM_Cubic>,<at 0.2254 7.5>,<at 0.4718 3.8333>,<at 0.8284 0.8667 RCIM_Cubic>]}
           GreenCurve {keys [<at 0 4.2112 RCIM_Cubic>,<at 0.2254 4>,<at 0.4718 1.5>,<at 0.8284 0.2 RCIM_Cubic>]}
           RedCurve {keys [<at 0 15 RCIM_Cubic>,<at 0.2254 10>,<at 0.4718 5>,<at 0.8284 1 RCIM_Cubic>]}
          }})
        2=?
        3=?
        Index=&RandomRangeInt(Minimum=0, Maximum=1)
        Num=::NewEnumerator0
       ))
      B=&MakeVectorFromLinearColorRGB(Color=&MergeHelper_ColorSwitch(
        0=&ColorFromCurve(ColorCurve=DataInterface() {propertyValues {
           AlphaCurve {keys [<at 0 1 RCIM_Cubic>,<at 1 0 RCIM_Cubic>]}
           BlueCurve {keys [<at 0 12 RCIM_Cubic>,<at 0.2254 8>,<at 0.4718 4>,<at 0.8284 0.526 RCIM_Cubic>]}
           GreenCurve {keys [<at 0 10 RCIM_Cubic>,<at 0.2254 6>,<at 0.4718 2.6>,<at 0.8284 0.2806 RCIM_Cubic>]}
           RedCurve {keys [<at 0 6 RCIM_Cubic>,<at 0.2254 3.2>,<at 0.4718 1.2>,<at 0.8284 0.1052 RCIM_Cubic>]}
          }})
        1=&ColorFromCurve(ColorCurve=DataInterface() {propertyValues {
           AlphaCurve {keys [<at 0 1 RCIM_Cubic>,<at 1 0 RCIM_Cubic>]}
           BlueCurve {keys [<at 0 12 RCIM_Cubic>,<at 0.2254 8>,<at 0.4718 4>,<at 0.8284 0.526 RCIM_Cubic>]}
           GreenCurve {keys [<at 0 5.1946 RCIM_Cubic>,<at 0.2254 3.0871>,<at 0.4718 1.7436>,<at 0.8284 0.1052 RCIM_Cubic>]}
           RedCurve {keys [<at 0 5.1946 RCIM_Cubic>,<at 0.2254 3.4965>,<at 0.4718 2.0445>,<at 0.8284 0.1543 RCIM_Cubic>]}
          }})
        2=?
        3=?
        Index=&RandomRangeInt(Minimum=0, Maximum=1)
        Num=::NewEnumerator0
       ))
      Alpha=$User.Variant
     )
     B=&MakeVectorFromLinearColorRGB(Color=$System.TODColorMultiplier)
    )
    ScaleA=true
    ScaleRGB=true
   ) {module=ScaleColor001}
   &DynamicMaterialParameters(
    Index 0 Param 1=&FloatFromCurve(
     FloatCurve=DataInterface() {propertyValues {Curve {keys [{interpMode=RCIM_Cubic, tangentMode=RCTM_User, time=0, value=0}, {interpMode=RCIM_Cubic, tangentMode=RCTM_User, time=1, value=1}]}}}
     Scale Curve=&RandomRangeFloat(Minimum=0.9, Maximum=1.1)
    )
    Module.Index0Float0Write=true
    Module.Index0Float1Write=true
    Module.Index0Float2Write=true
    Module.Index0Float3Write=true
   )
   &SkeletalMeshLocation(
    MaskSpawnGroup=false
    Mesh Sampling Type=::NewEnumerator5
    Sampled Position Offset=&SelectVectorFromArray(
     Array Sampling Mode=::NewEnumerator0
     Direct Array Index=?
     Vector Selection Array=DataInterface() {propertyValues {FloatData [
        <-8.4753 -7.4175 9.9954>
        <-7.1456 -8.2543 8.9485>
        <-10.1952 -7.7983 6.7771>
        <-11.4134 -7.5082 3.4077>
        <-11.0997 -7.4273 0.9736>
        <-9.804 -7.4134 -2.0411>
        <-9.5383 -7.5604 -1.7132>
        <-8.3465 -7.4109 -4.2224>
        <-9.2862 -7.8603 -1.389>
        <-4.9367 -7.9676 -6.0368>
        <-5.4021 -7.868 -5.7939>
        <-0.297 -8.1571 -7.9873>
        <0.3031 -7.8166 -8.7761>
        <3.8824 -7.4213 -7.9849>
        <5.8457 -7.5113 -5.9971>
        <5.1551 -8.0618 -5.6851>
        <5.386 -8.2403 -5.1652>
        <0.871 -8.3727 -7.376>
        <-1.2598 -8.6923 -6.4428>
        <-3.0679 -8.5994 -6.0455>
        <-0.9114 -8.9053 -5.8177>
        <3.2664 -9.1359 -4.1516>
        <2.159 -9.3432 -3.6286>
        <2.1072 -9.4666 -2.8754>
        <4.1865 -9.3643 -2.0519>
        <6.8742 -8.6798 -2.2716>
        <8.1559 -7.6145 -3.4604>
        <8.8345 -7.7477 -2.3179>
        <9.0474 -8.6388 4.1456>
        <9.6988 -8.2718 5.7324>
        <10.108 -8.0569 5.9044>
        <7.9733 -8.9409 4.5396>
        <8.9904 -8.4142 6.7329>
        <8.0495 -8.5549 7.4806>
        <8.3797 -8.5426 7.0185>
        <6.6057 -8.7306 7.7572>
        <6.9782 -8.4572 8.4044>
        <3.8177 -8.3258 9.3848>
        <-1.8586 -7.4821 10.6755>
        <-5.1542 -8.0878 9.8085>
        <-2.8125 -9.0099 7.4603>
        <-5.1119 -9.0474 7.0979>
        <-4.1683 -9.3063 6.107>
        <-1.5094 -9.2072 6.602>
        <-4.7615 -9.3739 5.4904>
        <-7.2822 -9.0694 4.9808>
        <-7.6062 -9.0489 3.742>
        <-8.6468 -8.6684 1.476>
        <-4.976 -9.5465 3.0565>
        <-8.118 -8.5417 -0.7919>
        <-7.4675 -8.8389 -0.2661>
        <-3.338 -8.5652 -5.9568>
        <-5.5991 -8.8262 -3.3498>
        <-3.2495 -9.4488 -2.3381>
        <-4.0491 -9.5311 -0.6036>
        <-3.1083 -9.7487 0.7856>
        <0.9954 -9.8869 1.6877>
        <-1.4964 -9.8023 3.3348>
        <-0.2111 -9.0547 7.1284>
        <-0.7933 -8.7113 8.1082>
        <1.4821 -8.516 8.7287>
        <2.6147 -8.9682 7.5623>
        <3.8251 -9.5533 -0.6768>
        <5.5684 -9.2112 -1.1841>
        <6.2143 -9.2919 1.282>
        <5.705 -9.4287 2.6305>
        <6.137 -9.2962 4.7471>
        <6.835 -8.9911 6.3559>
       ]}}
     Random Seed=0
    )
    Sampled Position Offset Coordinate Space:ENiagaraSkelMeshTransforms=::NewEnumerator2
    Skeletal Mesh=$System.DimeBlanketSkeletal
    Socket Sampling Mode=::NewEnumerator5
   )
  ]
  rendererClasses [&NiagaraSpriteRendererProperties]
  renderers [{rendererClass=&NiagaraSpriteRendererProperties}]
 }
 SparklesHeart001 {
  emitterUpdateScript [
   &EmitterState()
   &SpawnRate(
    Spawn Probability=&MakeCustomFloatFromBool(
     Bool=&SetBoolByFloatComparison(A=$System.Music_Reactive_FX_Preview, B=0)
     True Float=1.5
     False Float=$System.SpawnProbability
    )
    SpawnRate=8
    Use Spawn Probability=true
   )
  ]
  particleSpawnScript [
   &InitializeParticle(
    AdjustAlpha=true
    AdjustHue=true
    AdjustSaturation=true
    AdjustValue=true
    Color=&LerpLinearColors(StartColor=<rgba 1 0.6 0.6733 1>, EndColor=<rgba 0.5 0.9738 1 1>, LerpFactor=$User.Variant)
    Color Mode=::NewEnumerator3
    Hue Shift Range:Vector2f=<-0.025 0.025>
    Lifetime Max=0.4
    Lifetime Min=0.3
    Lifetime Mode=::NewEnumerator1
    Saturation Range:Vector2f=<1 1.2>
    Sprite Size Max=<35 35>
    Sprite Size Min=<25 25>
    Sprite Size Mode=::NewEnumerator2
    UsePositionOffset=false
    Write Material Random=true
   )
   &CameraOffset(Camera Offset Amount=&FN_FrontEndFloatModifier(Gameplay=5, Front End=2))
  ]
  particleUpdateScript [
   &ParticleState()
   &SkeletalMeshLocation(
    MaskSpawnGroup=false
    Mesh Sampling Type=::NewEnumerator5
    Sampled Position Offset=&SelectVectorFromArray(
     Array Sampling Mode=::NewEnumerator0
     Direct Array Index=?
     Vector Selection Array=DataInterface() {propertyValues {FloatData [
        <-8.4753 -7.4175 9.9954>
        <-7.1456 -8.2543 8.9485>
        <-10.1952 -7.7983 6.7771>
        <-11.4134 -7.5082 3.4077>
        <-11.0997 -7.4273 0.9736>
        <-9.804 -7.4134 -2.0411>
        <-9.5383 -7.5604 -1.7132>
        <-8.3465 -7.4109 -4.2224>
        <-9.2862 -7.8603 -1.389>
        <-4.9367 -7.9676 -6.0368>
        <-5.4021 -7.868 -5.7939>
        <-0.297 -8.1571 -7.9873>
        <0.3031 -7.8166 -8.7761>
        <3.8824 -7.4213 -7.9849>
        <5.8457 -7.5113 -5.9971>
        <5.1551 -8.0618 -5.6851>
        <5.386 -8.2403 -5.1652>
        <0.871 -8.3727 -7.376>
        <-1.2598 -8.6923 -6.4428>
        <-3.0679 -8.5994 -6.0455>
        <-0.9114 -8.9053 -5.8177>
        <3.2664 -9.1359 -4.1516>
        <2.159 -9.3432 -3.6286>
        <2.1072 -9.4666 -2.8754>
        <4.1865 -9.3643 -2.0519>
        <6.8742 -8.6798 -2.2716>
        <8.1559 -7.6145 -3.4604>
        <8.8345 -7.7477 -2.3179>
        <9.0474 -8.6388 4.1456>
        <9.6988 -8.2718 5.7324>
        <10.108 -8.0569 5.9044>
        <7.9733 -8.9409 4.5396>
        <8.9904 -8.4142 6.7329>
        <8.0495 -8.5549 7.4806>
        <8.3797 -8.5426 7.0185>
        <6.6057 -8.7306 7.7572>
        <6.9782 -8.4572 8.4044>
        <3.8177 -8.3258 9.3848>
        <-1.8586 -7.4821 10.6755>
        <-5.1542 -8.0878 9.8085>
        <-2.8125 -9.0099 7.4603>
        <-5.1119 -9.0474 7.0979>
        <-4.1683 -9.3063 6.107>
        <-1.5094 -9.2072 6.602>
        <-4.7615 -9.3739 5.4904>
        <-7.2822 -9.0694 4.9808>
        <-7.6062 -9.0489 3.742>
        <-8.6468 -8.6684 1.476>
        <-4.976 -9.5465 3.0565>
        <-8.118 -8.5417 -0.7919>
        <-7.4675 -8.8389 -0.2661>
        <-3.338 -8.5652 -5.9568>
        <-5.5991 -8.8262 -3.3498>
        <-3.2495 -9.4488 -2.3381>
        <-4.0491 -9.5311 -0.6036>
        <-3.1083 -9.7487 0.7856>
        <0.9954 -9.8869 1.6877>
        <-1.4964 -9.8023 3.3348>
        <-0.2111 -9.0547 7.1284>
        <-0.7933 -8.7113 8.1082>
        <1.4821 -8.516 8.7287>
        <2.6147 -8.9682 7.5623>
        <3.8251 -9.5533 -0.6768>
        <5.5684 -9.2112 -1.1841>
        <6.2143 -9.2919 1.282>
        <5.705 -9.4287 2.6305>
        <6.137 -9.2962 4.7471>
        <6.835 -8.9911 6.3559>
       ]}}
     Random Seed=0
    )
    Sampled Position Offset Coordinate Space:ENiagaraSkelMeshTransforms=::NewEnumerator2
    Skeletal Mesh=$System.DimeBlanketSkeletal
    Socket Sampling Mode=::NewEnumerator5
   )
   &ScaleSpriteSize(Uniform Curve Scale=&FNTimeOfDayScalar(Night Scale=1.25, Morning Scale=1, Evening Scale=1)) {module=ScaleSpriteSize001}
   &ScaleColor(
    Scale Alpha=&Multiply_Float(
     A=&FloatFromCurve(
      FloatCurve=DataInterface() {propertyValues {Curve {keys [
          {interpMode=RCIM_Cubic, tangentMode=RCTM_User, time=0, value=0}
          {arriveTangent=-2.4058, interpMode=RCIM_Cubic, leaveTangent=-2.4058, tangentMode=RCTM_User, time=0.26, value=0.8421}
          {arriveTangent=-1.1775, interpMode=RCIM_Cubic, leaveTangent=-1.1775, time=0.4613, value=0.0645}
          {arriveTangent=-0.1197, interpMode=RCIM_Cubic, leaveTangent=-0.1197, time=0.5898, value=0.4537}
          <at 1 0 RCIM_Cubic>
         ]}}}
      Scale Curve=&RandomRangeFloat(Minimum=0.1, Maximum=0.5)
     )
     B=&FloatFromCurve(
      FloatCurve=DataInterface() {propertyValues {Curve {keys [{tangentMode=RCTM_User, time=0, value=0}, {tangentMode=RCTM_User, time=1, value=1}]}}}
      Scale Curve=&FloatFromCurve(
       FloatCurve=DataInterface() {propertyValues {Curve {keys [
           {arriveTangentWeight=0.1667, interpMode=RCIM_Cubic, leaveTangentWeight=0.1667, tangentMode=RCTM_User, time=1.4, value=0}
           {tangentMode=RCTM_User, time=2, value=1}
          ]}}}
       CurveIndex=$Emitter.Age
      )
     )
    )
    Scale RGB=&VectorFromFloat(Value=&RandomRangeFloat(Minimum=1, Maximum=1.2, Evaluation Type=::NewEnumerator1))
    ScaleA=true
    ScaleRGB=true
   ) {module=ScaleColor001}
   &ScaleColor(
    Scale Alpha=&Multiply_Float(
     A=&Multiply_Float(
      A=&FloatFromCurve(
       FloatCurve=DataInterface() {propertyValues {Curve {keys [<at 0 1 RCIM_Cubic>,<at 1 0>]}}}
       Scale Curve=&RandomRangeFloat(Minimum=1, Maximum=2)
      )
      B=&MakeFloatFromLinearColor(LinearColor=$System.TODColorMultiplier, Channel:ENiagaraLinearColor_Channels=::NewEnumerator3)
     )
     B=&MakeCustomFloatFromBool(
      Bool=$User.MusicReactive
      True Float=1
      False Float=&MakeCustomFloatFromBool(
       Bool=&SetBoolByFloatComparison(A=$System.Music_Reactive_FX_Preview, B=0)
       True Float=&Multiply_Float(
        A=&Lerp_Float(
         A=&FN_FrontEndFloatModifier(
          Gameplay=&FN_PickFloatDependOnPlatform(TargetPlatformActive=1, TargetPlatformNoActive=0.8)
          Front End=1.25
         )
         B=&FN_FrontEndFloatModifier(
          Gameplay=&FN_PickFloatDependOnPlatform(TargetPlatformActive=3, TargetPlatformNoActive=1.5)
          Front End=2
         )
         Alpha=$System.Music_Reactive_FX_Preview
        )
        B=&FNTimeOfDayScalar(Night Scale=1, Morning Scale=1, Evening Scale=1)
       )
       False Float=&FN_FrontEndFloatModifier(
        Gameplay=1
        Front End=&FloatFromCurve(
         FloatCurve=DataInterface() {propertyValues {Curve {keys [
             <at 0 0 RCIM_Cubic>
             {arriveTangent=1.362, interpMode=RCIM_Cubic, leaveTangent=-0.9886, tangentMode=RCTM_Break, time=0.5, value=1.5}
             <at 1 0>
            ]}}}
         CurveIndex=&Cosine(
          Normalized Angle=&Add_Float(A=&Divide_Float(A=$Engine.Time, B=5), B=0.5)
          Period=1
          Scale=0.5
          Bias=1
         )
        )
       )
      )
     )
    )
    Scale RGB=&MakeVectorFromLinearColorRGB(Color=$System.TODColorMultiplier)
    ScaleA=true
    ScaleRGB=true
   )
   /SetVariables(
    Particles.CenterFlareColor:LinearColor=&MultiplyLinearColor(
     Linear Color=$Particles.Color
     Scale Factor=&MakeVector4_FromLinearColor(LinearColor=&LerpLinearColors(StartColor=<rgba 7 4.5194 4.9759 1>, EndColor=<rgba 3.4218 8.799 10 1>, LerpFactor=$User.Variant))
    )
    Particles.CenterFlareSize:Vector2f=&Multiply_Vector2(A=$Particles.SpriteSize, B=&Vector2DFromFloat(Value=0.5))
    Particles.CoronaFlareColor:LinearColor=&MultiplyLinearColor(
     Linear Color=$Particles.Color
     Scale Factor=&MakeVector4_FromLinearColor(LinearColor=&LerpLinearColors(StartColor=<rgba 0.15 0.15 0.15 1>, EndColor=<rgba 0.0729 0.0729 0.0729 1>, LerpFactor=$User.Variant))
    )
    Particles.CoronaFlareSize:Vector2f=&Multiply_Vector2(A=$Particles.Initial.SpriteSize, B=&Vector2DFromFloat(Value=0.75))
    Particles.FlickerFlareColor:LinearColor=&MultiplyLinearColor(
     Linear Color=$Particles.Color
     Scale Factor=&MakeVector4_FromLinearColor(LinearColor=&LerpLinearColors(StartColor=<rgba 1 0.9072 0.9536 1>, EndColor=<rgba 0.0393 0.0337 0.0417 50>, LerpFactor=$User.Variant))
    )
   ) {isSetParameters=true, module=SetVariables_#6}
  ]
  rendererClasses [&NiagaraSpriteRendererProperties]
  renderers [
   {rendererClass=&NiagaraSpriteRendererProperties}
   {rendererClass=&NiagaraSpriteRendererProperties, rendererIndex=1}
  ]
 }
 Sparkles_SG001 {
  emitterUpdateScript [
   &EmitterState()
   &SpawnRate(
    Spawn Probability=&MakeCustomFloatFromBool(
     Bool=&SetBoolByFloatComparison(A=$System.Music_Reactive_FX_Preview, B=0)
     True Float=1.5
     False Float=$System.SpawnProbability
    )
    SpawnRate=4
    Use Spawn Probability=true
   )
  ]
  particleSpawnScript [
   &InitializeParticle(
    AdjustAlpha=true
    AdjustHue=true
    AdjustSaturation=true
    AdjustValue=true
    Color=<rgba 2 2 2 1>
    Color Mode=::NewEnumerator3
    Lifetime Max=0.4
    Lifetime Min=0.3
    Lifetime Mode=::NewEnumerator1
    Sprite Size Max=<20 20>
    Sprite Size Min=<15 15>
    Sprite Size Mode=::NewEnumerator2
    UsePositionOffset=false
    Write Material Random=true
   )
   &CameraOffset(Camera Offset Amount=&FN_FrontEndFloatModifier(Gameplay=5, Front End=2))
  ]
  particleUpdateScript [
   &ParticleState()
   &SkeletalMeshLocation(
    Bone / Socket Index=&SelectIntFromArray(
     Array Sampling Mode=::NewEnumerator1
     Direct Array Index=?
     Int Selection Array=DataInterface() {propertyValues {IntData [2,3,4,5,6,7,8,9,10,11,12,13,14]}}
     Random Seed=0
    )
    MaskSpawnGroup=false
    Mesh Sampling Type=::NewEnumerator5
    Skeletal Mesh=$System.DimeBlanketSkeletal
    Socket Sampling Mode=::NewEnumerator5
   )
   &ScaleSpriteSize(
    Uniform Curve Scale=&FNTimeOfDayScalar(Night Scale=1.25, Morning Scale=1, Evening Scale=1)
    Uniform Curve Sprite Scale=DataInterface() {propertyValues {Curve {keys [<at 0 0 RCIM_Cubic>,<at 1 1 RCIM_Cubic>]}}}
   ) {module=ScaleSpriteSize001}
   &ScaleColor(
    Scale Alpha=&Multiply_Float(
     A=&FloatFromCurve(
      FloatCurve=DataInterface() {propertyValues {Curve {keys [
          {interpMode=RCIM_Cubic, tangentMode=RCTM_User, time=0, value=0}
          {arriveTangent=-2.4058, interpMode=RCIM_Cubic, leaveTangent=-2.4058, tangentMode=RCTM_User, time=0.26, value=0.8421}
          {arriveTangent=-1.1775, interpMode=RCIM_Cubic, leaveTangent=-1.1775, time=0.4613, value=0.0645}
          {arriveTangent=-0.1197, interpMode=RCIM_Cubic, leaveTangent=-0.1197, time=0.5898, value=0.4537}
          <at 1 0 RCIM_Cubic>
         ]}}}
      Scale Curve=&RandomRangeFloat(Minimum=0.1, Maximum=0.5)
     )
     B=&FloatFromCurve(
      FloatCurve=DataInterface() {propertyValues {Curve {keys [{tangentMode=RCTM_User, time=0, value=0}, {tangentMode=RCTM_User, time=1, value=1}]}}}
      Scale Curve=&FloatFromCurve(
       FloatCurve=DataInterface() {propertyValues {Curve {keys [
           {arriveTangentWeight=0.1667, interpMode=RCIM_Cubic, leaveTangentWeight=0.1667, tangentMode=RCTM_User, time=1.4, value=0}
           {tangentMode=RCTM_User, time=2, value=1}
          ]}}}
       CurveIndex=$Emitter.Age
      )
     )
    )
    Scale RGB=&VectorFromFloat(Value=&RandomRangeFloat(Minimum=0.7, Maximum=0.9, Evaluation Type=::NewEnumerator1))
    ScaleA=true
    ScaleRGB=true
   ) {module=ScaleColor001}
   &ScaleColor(
    Scale Alpha=&Multiply_Float(
     A=&Multiply_Float(
      A=&FloatFromCurve(
       FloatCurve=DataInterface() {propertyValues {Curve {keys [<at 0 1 RCIM_Cubic>,<at 1 0>]}}}
       Scale Curve=&RandomRangeFloat(Minimum=1, Maximum=2)
      )
      B=&MakeFloatFromLinearColor(LinearColor=$System.TODColorMultiplier, Channel:ENiagaraLinearColor_Channels=::NewEnumerator3)
     )
     B=&MakeCustomFloatFromBool(
      Bool=$User.MusicReactive
      True Float=1
      False Float=&MakeCustomFloatFromBool(
       Bool=&SetBoolByFloatComparison(A=$System.Music_Reactive_FX_Preview, B=0)
       True Float=&Multiply_Float(
        A=&Lerp_Float(
         A=&FN_FrontEndFloatModifier(
          Gameplay=&FN_PickFloatDependOnPlatform(TargetPlatformActive=1, TargetPlatformNoActive=0.8)
          Front End=1.5
         )
         B=&FN_FrontEndFloatModifier(
          Gameplay=&FN_PickFloatDependOnPlatform(TargetPlatformActive=5, TargetPlatformNoActive=1.5)
          Front End=2.5
         )
         Alpha=$System.Music_Reactive_FX_Preview
        )
        B=&FNTimeOfDayScalar(Night Scale=1, Morning Scale=1, Evening Scale=1)
       )
       False Float=&FN_FrontEndFloatModifier(
        Gameplay=1.5
        Front End=&FloatFromCurve(
         FloatCurve=DataInterface() {propertyValues {Curve {keys [<at 0 0 RCIM_Cubic>, {arriveTangent=1.362, interpMode=RCIM_Cubic, leaveTangent=-0.9886, tangentMode=RCTM_Break, time=0.5, value=2}, <at 1 0>]}}}
         CurveIndex=&Cosine(
          Normalized Angle=&Add_Float(A=&Divide_Float(A=$Engine.Time, B=5), B=0.5)
          Period=1
          Scale=0.5
          Bias=1
         )
        )
       )
      )
     )
    )
    Scale RGB=&MakeVectorFromLinearColorRGB(Color=$System.TODColorMultiplier)
    ScaleA=true
    ScaleRGB=true
   )
   /SetVariables(
    Particles.CenterFlareColor:LinearColor=&MultiplyLinearColor(
     Linear Color=$Particles.Color
     Scale Factor=&MakeVector4_FromLinearColor(LinearColor=&LerpLinearColors(StartColor=<rgba 10 9.4203 9.8068 1>, EndColor=<rgba 3.4218 8.799 10 1>, LerpFactor=$User.Variant))
    )
    Particles.CenterFlareSize:Vector2f=&Multiply_Vector2(A=$Particles.SpriteSize, B=&Vector2DFromFloat(Value=0.5))
    Particles.CoronaFlareColor:LinearColor=&MultiplyLinearColor(
     Linear Color=$Particles.Color
     Scale Factor=&MakeVector4_FromLinearColor(LinearColor=&LerpLinearColors(StartColor=<rgba 0.15 0.15 0.15 1>, EndColor=<rgba 0.0729 0.0729 0.0729 1>, LerpFactor=$User.Variant))
    )
    Particles.CoronaFlareSize:Vector2f=&Multiply_Vector2(A=$Particles.Initial.SpriteSize, B=&Vector2DFromFloat(Value=0.75))
    Particles.FlickerFlareColor:LinearColor=&MultiplyLinearColor(
     Linear Color=$Particles.Color
     Scale Factor=&MakeVector4_FromLinearColor(LinearColor=&LerpLinearColors(StartColor=<rgba 1 0.9826 1 1>, EndColor=<rgba 0.0393 0.0337 0.0417 50>, LerpFactor=$User.Variant))
    )
   ) {isSetParameters=true, module=SetVariables_#6}
  ]
  rendererClasses [&NiagaraSpriteRendererProperties]
  renderers [
   {rendererClass=&NiagaraSpriteRendererProperties}
   {rendererClass=&NiagaraSpriteRendererProperties, rendererIndex=1}
  ]
 }
 Sparkles_Silver001 {
  emitterUpdateScript [
   &EmitterState()
   &SpawnRate(
    Spawn Probability=&MakeCustomFloatFromBool(
     Bool=&SetBoolByFloatComparison(A=$System.Music_Reactive_FX_Preview, B=0)
     True Float=1.5
     False Float=$System.SpawnProbability
    )
    SpawnRate=12
    Use Spawn Probability=true
   )
  ]
  particleSpawnScript [
   &InitializeParticle(
    AdjustAlpha=true
    AdjustHue=true
    AdjustSaturation=true
    AdjustValue=true
    Color Mode=::NewEnumerator3
    Hue Shift Range:Vector2f=<0 0>
    Lifetime Max=0.4
    Lifetime Min=0.3
    Lifetime Mode=::NewEnumerator1
    Sprite Size Max=<35 35>
    Sprite Size Min=<25 25>
    Sprite Size Mode=::NewEnumerator2
    UsePositionOffset=false
    Write Material Random=true
   )
   &CameraOffset(Camera Offset Amount=&FN_FrontEndFloatModifier(Gameplay=1.5, Front End=2))
  ]
  particleUpdateScript [
   &ParticleState()
   &SkeletalMeshLocation(
    MaskSpawnGroup=false
    Mesh Sampling Type=::NewEnumerator5
    Sampled Position Offset=&SelectVectorFromArray(
     Array Sampling Mode=::NewEnumerator0
     Direct Array Index=?
     Vector Selection Array=DataInterface() {propertyValues {FloatData [
        <0.0857 -2.5554 -10.115>
        <0.0096 -4.1582 -10.1645>
        <0.5373 -5.9795 -9.9828>
        <0.2971 -6.597 -10.0684>
        <0.9473 -2.9242 -9.8226>
        <0.9198 -3.9806 -9.837>
        <1.4823 -4.24 -9.5537>
        <1.4654 -4.3295 -9.5622>
        <1.9367 -5.5482 -9.3125>
        <1.6835 -5.1066 -9.4523>
        <2.4626 -7.356 -8.997>
        <1.9865 -4.4225 -9.2846>
        <2.9083 -7.1519 -8.7533>
        <2.5784 -3.6441 -8.9523>
        <2.8154 -6.7735 -8.8134>
        <2.9034 -5.253 -8.7609>
        <3.1587 -6.3739 -8.6104>
        <3.7505 -6.9044 -8.2396>
        <3.5461 -4.3587 -8.3676>
        <4.5398 -6.1099 -7.7143>
        <5.0728 -7.3845 -7.3034>
        <4.6607 -5.7107 -7.6316>
        <4.8529 -4.2274 -7.4974>
        <5.631 -2.6584 -6.899>
        <5.5978 -5.2974 -6.9366>
        <5.5514 -6.1698 -6.972>
        <6.0283 -3.5482 -6.5825>
        <5.9302 -5.1781 -6.6653>
        <6.4407 -4.0085 -6.2382>
        <6.215 -5.0348 -6.4273>
        <6.6087 -6.3135 -6.0897>
        <6.7008 -5.4964 -6.0046>
        <6.936 -3.1569 -5.7873>
        <6.9239 -2.9395 -5.7983>
        <7.3741 -6.1065 -5.3827>
        <7.4752 -4.3887 -5.2798>
        <7.9031 -3.0715 -4.8474>
        <8.0324 -4.6202 -4.7178>
        <8.1063 -2.6412 -4.6257>
        <8.3114 -5.4307 -4.4166>
        <8.3468 -2.871 -4.3757>
        <8.444 -4.0816 -4.2682>
        <9.2836 -7.0669 -3.2782>
        <9.0712 -6.8357 -3.5389>
        <9.2559 -3.8555 -3.3101>
        <9.6543 -7.1244 -2.7852>
        <9.6525 -4.7714 -2.7916>
        <9.8583 -6.2186 -2.5141>
        <9.6457 -4.0221 -2.8001>
        <9.6503 -3.3055 -2.7934>
        <10.0219 -7.2476 -2.247>
        <10.1596 -5.0585 -2.079>
        <10.0361 -3.0813 -2.2622>
        <10.0309 -7.0567 -2.2732>
        <10.2264 -6.9826 -1.9806>
        <10.5537 -3.8205 -1.4738>
        <10.5405 -3.1175 -1.4955>
        <10.9935 -4.0933 -0.7026>
        <11.2464 -6.1178 -0.1852>
        <11.4701 -6.4 0.3697>
        <11.4534 -3.3949 0.33>
        <12.0743 -7.1469 3.8486>
        <12.0218 -7.1361 4.6115>
        <9.6005 -6.4167 9.6291>
        <9.808 -5.4697 9.4603>
        <9.7106 -5.1671 9.5619>
        <9.0817 -6.5935 10.0805>
        <9.3771 -3.5365 9.846>
        <8.6651 -2.9731 10.3606>
        <8.7873 -2.9872 10.2763>
        <7.9071 -2.9143 10.783>
        <7.9057 -6.5779 10.7855>
        <8.3353 -3.4858 10.5465>
        <8.2536 -2.9175 10.5913>
        <7.6422 -6.9181 10.8986>
        <8.1751 -2.5662 10.6111>
        <7.5925 -3.3616 10.9171>
        <7.4703 -4.754 10.9704>
        <9.5163 -3.2074 9.6999>
        <10.3296 -3.8775 8.8819>
        <6.2589 -3.1408 11.3877>
        <6.8606 -3.1305 11.2037>
        <6.9992 -5.1321 11.1609>
        <6.9911 -6.7909 11.1632>
        <5.5551 -2.9561 11.5375>
        <5.8581 -7.2282 11.4652>
        <5.487 -7.0457 11.5526>
        <5.3641 -4.9755 11.571>
        <4.9881 -4.9085 11.6068>
        <4.1715 -3.9772 11.6375>
        <3.9975 -4.9991 11.6354>
        <3.7566 -6.4543 11.6229>
        <3.3711 -5.3484 11.5776>
        <2.6449 -2.7199 11.4393>
        <3.1182 -5.3416 11.547>
        <2.5345 -6.479 11.4168>
        <2.4769 -4.9498 11.4043>
        <2.2386 -3.9968 11.3394>
        <2.3546 -3.0845 11.3761>
        <1.6587 -4.5642 11.152>
        <1.9353 -6.8558 11.2435>
        <1.2323 -4.3689 10.9868>
        <0.9941 -3.5179 10.8901>
        <1.3281 -6.6315 11.0261>
        <0.7362 -4.7149 10.7859>
        <0.6472 -6.5247 10.75>
        <0.0606 -5.9972 10.5087>
        <-12.2955 -4.6927 2.1671>
        <-12.181 -3.8981 1.603>
        <-12.0739 -5.5077 1.1049>
        <12.1216 -5.815 1.3097>
        <12.0644 -5.8115 1.0845>
        <13.0572 -5.96 4.7927>
        <12.962 -5.2669 4.5215>
        <-12.1357 -4.9247 7.2681>
        <-11.7562 -4.9743 8.3109>
        <-11.5673 -5.4242 9.1135>
        <-11.465 -5.2683 9.4333>
        <-11.1421 -5.933 10.2986>
        <-11.0137 -5.2578 10.5982>
        <-10.8133 -4.6364 11.0175>
        <-10.348 -4.6888 11.8815>
        <-10.1864 -4.7798 12.1276>
        <-10.1778 -5.6661 12.1407>
        <-9.6964 -5.9832 12.8731>
        <-9.6171 -5.7298 12.9806>
        <-9.5548 -6.0142 13.0571>
        <-8.2259 -4.1958 14.5466>
        <-8.4101 -5.9904 14.3587>
        <-7.4592 -4.1777 15.2367>
        <-7.4847 -5.1834 15.2142>
        <-6.9192 -6.0252 15.663>
        <-7.0224 -4.1062 15.5851>
        <-6.0143 -5.7589 16.2879>
        <-5.4069 -5.3019 16.6433>
        <-4.9828 -5.2512 16.8652>
        <-4.4265 -4.1788 17.1136>
        <-4.1947 -5.181 17.2119>
        <-3.3473 -5.6092 17.509>
        <-3.0602 -4.4365 17.5911>
        <-2.5349 -6.015 17.7271>
        <-2.2563 -6.0726 17.7862>
        <-1.4432 -5.2234 17.9249>
        <-1.1946 -4.2979 17.9533>
        <-0.4034 -5.2249 18.0131>
        <-0.0501 -3.9303 18.0269>
        <0.0951 -5.8164 18.0252>
        <0.2205 -4.3326 18.0203>
        <1.3982 -4.1811 17.93>
        <1.1784 -6.097 17.9552>
        <2.6612 -4.8666 17.6944>
        <2.818 -4.8079 17.6538>
        <3.8405 -4.7378 17.3435>
        <4.1858 -5.5256 17.2156>
        <4.5041 -4.701 17.0806>
        <5.0911 -5.9124 16.8085>
        <5.7284 -4.8949 16.469>
        <6.767 -6.0914 15.7779>
        <7.3791 -5.4536 15.3072>
        <8.0039 -4.2811 14.7571>
        <8.2635 -5.0294 14.5083>
        <8.865 -5.6921 13.8948>
        <9.2879 -4.6892 13.3847>
        <9.6068 -5.5718 12.9932>
        <10.0536 -5.6175 12.3297>
        <10.2989 -4.4109 11.9565>
        <10.383 -5.1241 11.8165>
        <10.7661 -4.8151 11.1051>
        <11.1152 -4.5355 10.3612>
        <11.2609 -5.8105 10.0215>
        <11.4471 -5.157 9.4892>
        <11.4333 -4.6819 9.5323>
        <11.6715 -5.2911 8.6843>
        <12.0259 -4.2269 7.4134>
        <12.1545 -5.374 7.2432>
        <-12.9201 -5.3563 3.4834>
        <-12.6272 -4.0435 2.205>
        <-0.1937 -7.072 10.581>
        <-0.127 -3.9863 10.5473>
        <-0.4686 -4.7852 10.681>
        <-0.5393 -3.3881 10.7083>
        <-1.373 -6.6742 11.0454>
        <-1.0481 -4.9362 10.9125>
        <-1.1286 -2.7571 10.9436>
        <-1.959 -5.2135 11.2515>
        <-1.8993 -6.018 11.2317>
        <-1.9208 -4.2786 11.2383>
        <-1.7601 -2.5085 11.1617>
        <-2.8965 -7.1137 11.495>
        <-2.6842 -5.2785 11.4496>
        <-2.6146 -5.0253 11.4339>
        <-3.0117 -4.7098 11.5224>
        <-3.1943 -4.5617 11.5577>
        <-3.4852 -2.8744 11.5906>
        <-3.8371 -5.4012 11.6305>
        <-3.9195 -2.8775 11.6322>
        <-4.9488 -6.7344 11.6101>
        <-4.7495 -4.4506 11.6285>
        <-5.0917 -3.8343 11.5969>
        <-5.328 -5.9806 11.577>
        <-6.0024 -6.8966 11.4517>
        <-5.9373 -6.8651 11.4643>
        <-5.6871 -3.6407 11.5124>
        <-6.5803 -7.15 11.2825>
        <-6.6542 -2.8027 11.2648>
        <-6.461 -3.7843 11.3233>
        <-6.4768 -6.1991 11.3194>
        <-7.7166 -3.6433 10.8647>
        <-7.2433 -3.6012 11.0645>
        <-7.783 -6.9991 10.8384>
        <-7.8696 -6.3427 10.8013>
        <-8.0877 -7.2542 10.6685>
        <-8.0794 -6.8473 10.6903>
        <-8.2797 -6.3355 10.5801>
        <-8.2682 -3.1024 10.5848>
        <-8.4077 -5.5845 10.5094>
        <-8.7876 -2.5514 10.2422>
        <-8.6684 -4.3356 10.3509>
        <-9.1481 -4.0591 10.0221>
        <-8.9626 -3.0761 10.1485>
        <-9.4276 -6.0295 9.7901>
        <-9.5575 -5.005 9.6791>
        <-9.6362 -3.895 9.6113>
        <-10.084 -6.5573 9.1601>
        <-10.1867 -3.7036 9.0496>
        <-10.4922 -5.6756 8.671>
        <-10.4682 -7.2788 8.6679>
        <-10.6766 -2.91 8.4348>
        <-11.0253 -7.1041 7.8925>
        <-11.0665 -2.9701 7.8233>
        <-10.846 -3.5197 8.1753>
        <-11.6794 -2.584 6.2247>
        <-11.8126 -2.5359 5.6477>
        <-12.03 -7.2045 2.8973>
        <-11.7917 -5.2657 1.3792>
        <-11.7248 -6.3518 1.1464>
        <-11.702 -3.9457 1.0709>
        <-11.4274 -3.0193 0.2605>
        <-11.4533 -6.4069 0.3251>
        <-11.1454 -4.8706 -0.3935>
        <-10.9287 -7.132 -0.8247>
        <-10.9626 -6.7224 -0.7699>
        <-11.0002 -4.3596 -0.6919>
        <-10.7831 -7.2584 -1.0385>
        <-10.5591 -3.1905 -1.467>
        <-10.7489 -5.6753 -1.1442>
        <-10.5041 -3.3821 -1.561>
        <-10.1926 -4.2532 -2.0307>
        <-10.2172 -5.9267 -1.9945>
        <-9.7273 -2.6732 -2.6755>
        <-9.9163 -5.3861 -2.4357>
        <-9.6789 -5.334 -2.7569>
        <-9.7111 -5.3344 -2.7133>
        <-9.3259 -5.7583 -3.2256>
        <-9.3392 -5.7063 -3.2093>
        <-9.178 -6.1815 -3.4074>
        <-8.9075 -6.2906 -3.7395>
        <-8.6848 -3.3118 -3.9996>
        <-8.6786 -3.3376 -4.0065>
        <-8.5434 -5.7406 -4.1585>
        <-8.3956 -5.3696 -4.3229>
        <-7.9757 -3.2183 -4.7742>
        <-8.1045 -3.0071 -4.6441>
        <-7.386 -6.9642 -5.3711>
        <-7.9992 -6.8625 -4.7525>
        <-7.2396 -5.7753 -5.5097>
        <-7.2836 -5.933 -5.4693>
        <-6.9223 -6.3777 -5.8016>
        <-6.554 -4.7242 -6.1392>
        <-5.8204 -6.8668 -6.7581>
        <-6.184 -6.5112 -6.4541>
        <-5.7837 -4.5173 -6.7877>
        <-5.9918 -3.9146 -6.6135>
        <-5.5025 -5.632 -7.0081>
        <-5.4221 -3.7705 -7.0679>
        <-5.3816 -2.5372 -7.0699>
        <-4.9136 -4.0282 -7.4516>
        <-4.9502 -5.5907 -7.4246>
        <-4.7994 -2.5858 -7.5178>
        <-4.6337 -3.575 -7.649>
        <-4.7086 -5.9084 -7.5989>
        <-4.23 -4.2267 -7.9251>
        <-4.0741 -7.0578 -8.0326>
        <-3.6151 -3.3391 -8.323>
        <-3.3861 -3.3796 -8.4672>
        <-3.573 -6.2584 -8.3508>
        <-3.6156 -5.4918 -8.3237>
        <-3.0885 -4.29 -8.652>
        <-2.6957 -3.0003 -8.8827>
        <-2.7023 -6.4013 -8.8802>
        <-2.4058 -6.3734 -9.0542>
        <-2.3037 -3.0704 -9.1091>
        <-1.7719 -2.4706 -9.3719>
        <-2.3366 -6.9411 -9.0926>
        <-1.8927 -7.2024 -9.3273>
        <-1.1978 -2.8186 -9.6962>
        <-1.0437 -3.9907 -9.7742>
        <-0.72 -4.7452 -9.9194>
        <-0.5082 -6.7586 -9.9946>
        <-0.2473 -5.4234 -10.086>
       ]}}
     Random Seed=0
    )
    Skeletal Mesh=$System.DimeBlanketSkeletal
    Socket Sampling Mode=::NewEnumerator5
   )
   &ScaleSpriteSize(
    Uniform Curve Scale=&FNTimeOfDayScalar(Night Scale=1, Morning Scale=1, Evening Scale=1)
    Uniform Curve Sprite Scale=DataInterface() {propertyValues {Curve {keys [<at 0 0 RCIM_Cubic>,<at 1 1 RCIM_Cubic>]}}}
   ) {module=ScaleSpriteSize001}
   &ScaleColor(
    Scale Alpha=&Multiply_Float(
     A=&FloatFromCurve(
      FloatCurve=DataInterface() {propertyValues {Curve {keys [
          {interpMode=RCIM_Cubic, tangentMode=RCTM_User, time=0, value=0}
          {arriveTangent=-2.4058, interpMode=RCIM_Cubic, leaveTangent=-2.4058, tangentMode=RCTM_User, time=0.26, value=0.8421}
          {arriveTangent=-1.1775, interpMode=RCIM_Cubic, leaveTangent=-1.1775, time=0.4613, value=0.0645}
          {arriveTangent=-0.1197, interpMode=RCIM_Cubic, leaveTangent=-0.1197, time=0.5898, value=0.4537}
          <at 1 0 RCIM_Cubic>
         ]}}}
      Scale Curve=&RandomRangeFloat(Minimum=0.1, Maximum=0.5)
     )
     B=&FloatFromCurve(
      FloatCurve=DataInterface() {propertyValues {Curve {keys [{tangentMode=RCTM_User, time=0, value=0}, {tangentMode=RCTM_User, time=1, value=1}]}}}
      Scale Curve=&FloatFromCurve(
       FloatCurve=DataInterface() {propertyValues {Curve {keys [
           {arriveTangentWeight=0.1667, interpMode=RCIM_Cubic, leaveTangentWeight=0.1667, tangentMode=RCTM_User, time=1.4, value=0}
           {tangentMode=RCTM_User, time=2, value=1}
          ]}}}
       CurveIndex=$Emitter.Age
      )
     )
    )
    Scale RGB=&VectorFromFloat(Value=&RandomRangeFloat(Minimum=0.7, Maximum=0.9, Evaluation Type=::NewEnumerator1))
    ScaleA=true
    ScaleRGB=true
   ) {module=ScaleColor001}
   &ScaleColor(
    Scale Alpha=&Multiply_Float(
     A=&Multiply_Float(
      A=&FloatFromCurve(
       FloatCurve=DataInterface() {propertyValues {Curve {keys [<at 0 1 RCIM_Cubic>,<at 1 0>]}}}
       Scale Curve=&RandomRangeFloat(Minimum=1, Maximum=1.5)
      )
      B=&MakeFloatFromLinearColor(LinearColor=$System.TODColorMultiplier, Channel:ENiagaraLinearColor_Channels=::NewEnumerator3)
     )
     B=&MakeCustomFloatFromBool(
      Bool=$User.MusicReactive
      True Float=1
      False Float=&MakeCustomFloatFromBool(
       Bool=&SetBoolByFloatComparison(A=$System.Music_Reactive_FX_Preview, B=0)
       True Float=&Multiply_Float(
        A=&Lerp_Float(
         A=&FN_FrontEndFloatModifier(
          Gameplay=&FN_PickFloatDependOnPlatform(TargetPlatformActive=2, TargetPlatformNoActive=0.8)
          Front End=1
         )
         B=&FN_FrontEndFloatModifier(
          Gameplay=&FN_PickFloatDependOnPlatform(TargetPlatformActive=2, TargetPlatformNoActive=1.5)
          Front End=2
         )
         Alpha=$System.Music_Reactive_FX_Preview
        )
        B=&FNTimeOfDayScalar(Night Scale=1, Morning Scale=1, Evening Scale=1)
       )
       False Float=&FN_FrontEndFloatModifier(
        Gameplay=1
        Front End=&FloatFromCurve(
         FloatCurve=DataInterface() {propertyValues {Curve {keys [
             <at 0 0 RCIM_Cubic>
             {arriveTangent=1.362, interpMode=RCIM_Cubic, leaveTangent=-0.9886, tangentMode=RCTM_Break, time=0.5, value=1.5}
             <at 1 0>
            ]}}}
         CurveIndex=&Cosine(
          Normalized Angle=&Add_Float(A=&Divide_Float(A=$Engine.Time, B=5), B=0.5)
          Period=1
          Scale=0.5
          Bias=1
         )
        )
       )
      )
     )
    )
    Scale RGB=&MakeVectorFromLinearColorRGB(Color=$System.TODColorMultiplier)
    ScaleA=true
    ScaleRGB=true
   )
   /SetVariables(
    Particles.CenterFlareColor:LinearColor=&MultiplyLinearColor(
     Linear Color=$Particles.Color
     Scale Factor=&MakeVector4_FromLinearColor(LinearColor=&LerpLinearColors(StartColor=<rgba 10 9.7681 9.9227 1>, EndColor=<rgba 3.4218 8.799 10 1>, LerpFactor=$User.Variant))
    )
    Particles.CenterFlareSize:Vector2f=&Multiply_Vector2(A=$Particles.SpriteSize, B=&Vector2DFromFloat(Value=0.5))
    Particles.CoronaFlareColor:LinearColor=&MultiplyLinearColor(
     Linear Color=$Particles.Color
     Scale Factor=&MakeVector4_FromLinearColor(LinearColor=&LerpLinearColors(StartColor=<rgba 0.15 0.15 0.15 1>, EndColor=<rgba 0.0729 0.0729 0.0729 1>, LerpFactor=$User.Variant))
    )
    Particles.CoronaFlareSize:Vector2f=&Multiply_Vector2(A=$Particles.Initial.SpriteSize, B=&Vector2DFromFloat(Value=0.75))
    Particles.FlickerFlareColor:LinearColor=&MultiplyLinearColor(
     Linear Color=$Particles.Color
     Scale Factor=&MakeVector4_FromLinearColor(LinearColor=&LerpLinearColors(StartColor=<rgba 1 0.9362 1 1>, EndColor=<rgba 0.0393 0.0337 0.0417 50>, LerpFactor=$User.Variant))
    )
   ) {isSetParameters=true, module=SetVariables_#6}
  ]
  rendererClasses [&NiagaraSpriteRendererProperties]
  renderers [
   {rendererClass=&NiagaraSpriteRendererProperties}
   {rendererClass=&NiagaraSpriteRendererProperties, rendererIndex=1}
  ]
 }
}
summary {
 systemName=NS_Backpack_DimeBlanketGrace
 emitters [
  {emitterName=Hearts001, rendererClasses [&NiagaraSpriteRendererProperties]}
  {emitterName=SparklesHeart001, rendererClasses [&NiagaraSpriteRendererProperties]}
  {emitterName=Sparkles_Silver001, rendererClasses [&NiagaraSpriteRendererProperties]}
  {emitterName=Sparkles_SG001, rendererClasses [&NiagaraSpriteRendererProperties]}
 ]
 userVariables [
  User.bIsFrontEnd {type=NiagaraBool}
  User.bIsFrontEndPreview {type=NiagaraBool}
  User.bPickaxeEquipped {type=NiagaraBool}
  User.Lifetime_Scale {type=NiagaraFloat, defaultValue=0, description=""}
  User.MusicReactive {type=NiagaraBool}
  User.Variant {type=NiagaraFloat, defaultValue=0, description=""}
 ]
}
