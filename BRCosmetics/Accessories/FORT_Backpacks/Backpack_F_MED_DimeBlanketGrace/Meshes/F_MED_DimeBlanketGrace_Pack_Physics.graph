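# Physics asset for the DimeBlanketGrace backpack cosmetic, setting up a dangling keychain that swings from the pack.
# Two capsules form the backpack body, and the two-bone keychain chain uses small capsules, with a longer angled one on the end bone.
# Constraints link the first keychain bone to the backpack and the second to the first, allowing moderate, asymmetric swing and very little twist.
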PhysicsAsset /BRCosmetics/Accessories/FORT_Backpacks/Backpack_F_MED_DimeBlanketGrace/Meshes/F_MED_DimeBlanketGrace_Pack_Physics
bodies {
 backpack [
  {shapeType=Capsule, length=13, radius=3, center=<2.2898 -2.6499 -3.1044>, rotation=<rot 0 -0.0006 -14.8833>}
  {
   shapeType=Capsule
   length=13
   radius=3
   shapeName=backpack_capsule
   center=<-2.4659 -2.4249 -3.3524>
   rotation=<rot 18.2792 -3.4644 -11.6223>
  }
 ]
 keychain_dyn_01 [{shapeType=Capsule, length=1, radius=0.9, shapeName=keychain_dyn_01_capsule, center=<0.7 0 0>, rotation=<rot 90 0 0>}]
 keychain_dyn_02 [
  {shapeType=Capsule, length=1, radius=1, shapeName=keychain_dyn_02_capsule, center=<1.1 0 0>, rotation=<rot 90 0 0>}
  {
   shapeType=Capsule
   length=7
   radius=1.3
   shapeName=keychain_dyn_02_capsule
   center=<2.7639 -2.0497 1.8019>
   rotation=<rot -13.3976 46.1648 -71.6086>
  }
 ]
}
constraints [
 {bone1Name=keychain_dyn_01, bone2Name=backpack, swing1LimitDegrees=15, swing2LimitDegrees=45, twistLimitDegrees=3}
 {bone1Name=keychain_dyn_02, bone2Name=keychain_dyn_01, swing1LimitDegrees=15, swing2LimitDegrees=45, twistLimitDegrees=5}
]
