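# Anim Blueprint for the DimeBlanketGrace backpack cosmetic, adding secondary motion to its keychain and handle.
# With rigid body enabled, the keychain uses AnimDynamics in the front end and RigidBody simulation otherwise; when disabled, it takes a fixed offset pose.
# The cached result feeds handle AnimDynamics, then an additive pack offset mapped from spine_05-to-spine_01 yaw.
# The EventGraph also sets per-state gravity and sim-space variables and zipline flags.
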AnimBlueprint /BRCosmetics/Accessories/FORT_Backpacks/Backpack_F_MED_DimeBlanketGrace/Meshes/F_MED_DimeBlanketGrace_Pack_AnimBP
target_skeleton=&/F_MED_DimeBlanketGrace_Pack_Skeleton
graphs [
 AnimGraph {nodes [
   n1 = Anim Dynamics - Bone: handle_dyn_01!(
    ComponentPose=Local To Component(LocalPose=Use cached pose 'RBAN Pose cache'())
    Alpha=0
    GravityOverride:Vector="0,0,-980"
    ExternalForce="0,0,0"
   ) -> (Pose) {
    angularDampingOverride=1
    angularSpringConstant=23000
    bAngularSpring=true
    bOverrideAngularDamping=true
    gravityScale=0.35
    linearSpringConstant=10000
    simulationSpace=RootRelative
    alphaBoolBlend {blendOption=Linear}
    boundBone {boneName=handle_dyn_01}
    componentAppliedLinearAccClamp=<3000 3000 3000>
    componentLinearAccScale=<0.4 0.4 0.4>
    componentLinearVelScale=<0.4 0.4 0.4>
    gravityOverride {z=-980}
    physicsBodyDefinitions [{
      boundBone {boneName=handle_dyn_01}
      boxExtents=<34 34 34>
      constraintSetup {angularLimitsMax=<0 0 16>, angularLimitsMin=<0 0 -16>, angularTarget=<1 0 0>}
      localJointOffset=<3 0 0>
     }]
   }
   n2 = Anim Dynamics - Bone: handle_dyn_02!(ComponentPose=@n1, Alpha=1, GravityOverride:Vector="0,0,-980", ExternalForce="0,0,0") -> (Pose) {
    angularDampingOverride=1
    angularSpringConstant=27000
    bAngularSpring=true
    bOverrideAngularDamping=true
    bOverrideLinearDamping=true
    gravityScale=0.35
    linearDampingOverride=1
    simulationSpace=RootRelative
    alphaBoolBlend {blendOption=Linear}
    boundBone {boneName=handle_dyn_02}
    componentAppliedLinearAccClamp=<4000 4000 4000>
    componentLinearAccScale=<0.4 0.4 0.4>
    componentLinearVelScale=<0.4 0.4 0.4>
    gravityOverride {z=-980}
    physicsBodyDefinitions [{
      boundBone {boneName=handle_dyn_02}
      boxExtents=<40 40 40>
      constraintSetup {angularLimitsMax=<0 0 3>, angularLimitsMin=<0 0 -3>, angularTarget=<1 0 0>, linearAxesMax=<0.7 0.25 2.75>, linearAxesMin=<0 -0.25 -2.75>}
      localJointOffset=<3 0 0>
     }]
   }
   n3 = "Transform (Modify) Bone - Bone: pack"(ComponentPose=@n2, Translation=Get backpack_translation(), Rotation=Get backpack_rotate()) {rotationMode=BMM_Additive, translationMode=BMM_Additive, alphaBoolBlend {blendOption=Linear}, boneToModify {boneName=pack}}
   Output Pose(Result=Component To Local(ComponentPose=@n3)) {layerGroup=None}
   Save cached pose 'StartPose'(Pose=Input Pose()) {cacheName=StartPose}
   n4 = "Transform (Modify) Bone - Bone: keychain_base"(
    ComponentPose=Local To Component(LocalPose=Use cached pose 'StartPose'())
    Alpha=1
    Translation="0,0,0"
    Rotation="20.407,-12.694,42.239"
    Scale="1,1,1"
   ) {
    rotationMode=BMM_Additive
    rotationSpace=BCS_BoneSpace
    alphaBoolBlend {blendOption=Linear}
    boneToModify {boneName=keychain_base}
    rotation=<rot 20.407 -12.694 42.239>
   }
   n5 = Anim Dynamics - Bone: keychain_dyn_01!(ComponentPose=@n4, Alpha=1, GravityOverride:Vector="0,0,-980", ExternalForce="0,0,-980") -> (Pose) {
    angularDampingOverride=1
    angularSpringConstant=3000
    bOverrideAngularDamping=true
    bUseGravityOverride=true
    gravityScale=0.35
    alphaBoolBlend {blendOption=Linear}
    boundBone {boneName=keychain_dyn_01}
    componentAppliedLinearAccClamp=<3000 3000 3000>
    componentLinearAccScale=<0.4 0.4 0.4>
    componentLinearVelScale=<0.4 0.4 0.4>
    gravityOverride {z=-980}
    physicsBodyDefinitions [{
      boundBone {boneName=keychain_dyn_01}
      boxExtents=<30 30 30>
      constraintSetup {angularLimitsMax=<2 25 60>, angularLimitsMin=<-2 -50 0>, angularTarget=<1 0 0>}
      localJointOffset=<4 0 0>
     }]
   }
   n6 = Anim Dynamics - Bone: keychain_dyn_02!(ComponentPose=@n5, Alpha=1, GravityOverride:Vector="0,0,-980", ExternalForce="0,0,-980") -> (Pose) {
    angularDampingOverride=1
    angularSpringConstant=3000
    bOverrideAngularDamping=true
    bUseGravityOverride=true
    gravityScale=0.35
    alphaBoolBlend {blendOption=Linear}
    boundBone {boneName=keychain_dyn_02}
    componentAppliedLinearAccClamp=<3000 3000 3000>
    componentLinearAccScale=<0.4 0.4 0.4>
    componentLinearVelScale=<0.4 0.4 0.4>
    gravityOverride {z=-980}
    physicsBodyDefinitions [{
      boundBone {boneName=keychain_dyn_02}
      boxExtents=<30 30 30>
      constraintSetup {angularLimitsMax=<2 50 70>, angularLimitsMin=<-2 -60 0>, angularTarget=<1 0 0>}
      localJointOffset=<4 -2 2>
     }]
   }
   n7 = Component To Local(ComponentPose=@n6)
   n8 = "Transform (Modify) Bone - Bone: keychain_base"(ComponentPose=Local To Component(LocalPose=Use cached pose 'StartPose'()), Alpha=1, Translation="0,0,0", Rotation="0,20,25", Scale="1,1,1") {
    rotationMode=BMM_Additive
    rotationSpace=BCS_BoneSpace
    alphaBoolBlend {blendOption=Linear}
    boneToModify {boneName=keychain_base}
    rotation {roll=25, yaw=20}
   }
   n9 = RigidBody!(
    ComponentPose=@n8
    Alpha
    OverrideWorldGravity:Vector="0,0,0"
    ExternalForce
    SimSpaceSettings:Sim Space Settings Structure=(WorldAlpha=0.7,VelocityScaleZ=0,DampingAlpha=1,MaxLinearVelocity=700,MaxAngularVelocity=3,MaxLinearAcceleration=600,MaxAngularAcceleration=1000,ExternalLinearDragV=(X=0.2,Y=0.2,Z=0),ExternalLinearVelocity=(X=0,Y=0,Z=0),ExternalAngularVelocity=(X=0,Y=0,Z=0))
   ) -> (Pose) {
    bOverrideWorldGravity=true
    overridePhysicsAsset=&/F_MED_DimeBlanketGrace_Pack_Physics
    simulationSpace=BaseBoneSpace
    alphaBoolBlend {blendOption=Linear}
    baseBoneRef {boneName=root}
    componentAppliedLinearAccClamp=<3000 3000 3000>
    componentLinearAccScale=<0.675 0.675 0.675>
    componentLinearVelScale=<0.675 0.675 0.675>
    simSpaceSettings {
     maxAngularAcceleration=1000
     maxAngularVelocity=3
     maxLinearAcceleration=600
     maxLinearVelocity=700
     velocityScaleZ=0
     worldAlpha=0.7
     externalLinearDragV=<0.2 0.2>
    }
   }
   n10 = "Transform (Modify) Bone - Bone: pack"!(ComponentPose=@n9, bAlphaBoolEnabled:bool, Translation="0,-0.75,0", Rotation="0,0,0", Scale) -> (Pose) {
    alphaInputType=Bool
    translationMode=BMM_Additive
    alphaBoolBlend {blendInTime=1, blendOption=Sinusoidal, blendOutTime=1}
    boneToModify {boneName=pack}
    translation {y=-0.75}
   }
   n11 = Component To Local(ComponentPose=@n10)
   n12 = Get bIsInFrontEnd()
   n13 = "Transform (Modify) Bone - Bone: keychain_dyn_01"(
    ComponentPose=Local To Component(LocalPose=Use cached pose 'StartPose'())
    Alpha=1
    Translation="0,0,0"
    Rotation="-18.517,-1.211,21.595"
    Scale="1,1,1"
   ) {
    rotationMode=BMM_Additive
    rotationSpace=BCS_BoneSpace
    alphaBoolBlend {blendOption=Linear}
    boneToModify {boneName=keychain_dyn_01}
    rotation=<rot -18.5171 -1.211 21.595>
   }
   n14 = "Transform (Modify) Bone - Bone: keychain_dyn_02"(ComponentPose=@n13, Alpha=1, Translation="0,0,0", Rotation="-31.557,23.273,29.791", Scale="1,1,1") {
    rotationMode=BMM_Additive
    rotationSpace=BCS_BoneSpace
    alphaBoolBlend {blendOption=Linear}
    boneToModify {boneName=keychain_dyn_02}
    rotation=<rot -31.557 23.273 29.791>
   }
   Save cached pose 'RBAN Pose cache'(Pose=Blend Poses by bool(
     BlendPose_0=Blend Poses by bool(BlendPose_0=@n7, BlendPose_1=@n11, bActiveValue=@n12) {blendType=Linear}
     BlendPose_1=Component To Local(ComponentPose=@n14)
     bActiveValue=Get bRigidBodyEnabled()
    ) {blendType=Linear}) {cacheName=RBAN Pose cache}
  ]}
 EventGraph {nodes [
   Event BlueprintUpdateAnimation()
   > n1 = Sequence()
   > n2 = Gravity Override State Function DEPRECATED(
    CharacterPartAnimInstance=Self-Reference()
    GameTimeInSeconds=GetGameTimeInSeconds()
    GravityOverrideParams=Get Gravity Override Params()
    RBAN Params=Get RBAN Params()
   )
   > Set GravityOverride(GravityOverride=@n2)
   > Set Sim Space Settings(Sim Space Settings:Sim Space Settings Structure=Break RBANSetting Struct!(RBANSettingStruct:"RBANSetting Struct Structure (by ref)"=@n2.RBAN SettingsOut) -> (
     ComponentLinearAccScale_28_#1:Vector=~
     ComponentLinearVelScale_31_#2:Vector=~
     SimSpaceSettings_34_#3:Sim Space Settings Structure
    ) {is=title K2Node_BreakStruct}.SimSpaceSettings_34_#3) -> (Output_Get:Sim Space Settings Structure)
   n3 = GetOwnerSkeletalMeshComponent()
   n4 = BreakRotator(InRot=BreakTransform(InTransform=GetDeltaTransformFromRefPose(self=@n3, BoneName=spine_05, BaseName=spine_01)).Rotation)
   > Set backpack_rotate(backpack_rotate=MakeRotator(Roll=MapRangeClamped(Value=@n4.Yaw, InRangeA=0, InRangeB=-30, OutRangeB=30)))
   > n5 = Set backpack_translation(backpack_translation=MakeVector(Y=MapRangeClamped(Value=@n4.Yaw, InRangeA=0, InRangeB=-30, OutRangeA=0.5, OutRangeB=-4)))
   n6 = Event BlueprintProcessAnimInputEvent()
   n7 = Branch(execute=@n5.then, Condition=IsNamed(self=@n6.AnimInputEvent, InEventName=aie.Player.Zipline.Entered))
   > Set OnZipline(OnZipline:bool=1) -> (Output_Get:bool)
   > Set ZiplineAlpha(ZiplineAlpha:double=0.15) -> (Output_Get:double)
   Branch(
    execute=@n7.else
    Condition=OR Boolean!(A=IsNamed(self=@n6.AnimInputEvent, InEventName=aie.Player.Zipline.Exited), B=Get bIsMoving2D(), C:bool=Get bIsJumping()) -> (ReturnValue)
   )
   > Set OnZipline(OnZipline=false)
   > Set ZiplineAlpha(ZiplineAlpha=1)
   n8 = Branch(execute=@n1.then_1, Condition=Get bRigidBodyEnabled())
   > Set Is RBAN(Is RBAN:double=0) -> (Output_Get:double)
   Set Is RBAN(execute=@n8.else, Is RBAN=1)
   TryGetPawnOwner()
  ]}
]
variables [
 backpack_rotate {type=Rotator}
 backpack_translation {type=Vector}
 Gravity Override Params {type=GravityOverrideParamsStruct}
 GravityOverride {type=Vector}
 Is RBAN {type=number}
 OnZipline {type=boolean}
 RBAN Params {
  type=RBANParamsStruct
  default {
   dBNO {simSpaceSettings {
     maxAngularAcceleration=1000
     maxAngularVelocity=3
     maxLinearAcceleration=600
     maxLinearVelocity=700
     velocityScaleZ=0
     worldAlpha=0.5
     externalLinearDragV=<0.2 0.2 0>
    }}
   dBNO_Carried {simSpaceSettings {
     maxAngularAcceleration=1000
     maxAngularVelocity=3
     maxLinearAcceleration=600
     maxLinearVelocity=700
     velocityScaleZ=0
     worldAlpha=0.5
     externalLinearDragV=<0.2 0.2 0>
    }}
   emoteOrMelee {simSpaceSettings {
     maxAngularAcceleration=1000
     maxAngularVelocity=3
     maxLinearAcceleration=600
     maxLinearVelocity=700
     velocityScaleZ=0
     worldAlpha=0.5
     externalLinearDragV=<0.2 0.2 0>
    }}
   front_End {simSpaceSettings {
     maxAngularAcceleration=1000
     maxAngularVelocity=3
     maxLinearAcceleration=600
     maxLinearVelocity=700
     velocityScaleZ=1
     externalLinearDragV=<0.2 0.2 0.3>
    }}
   onGround_Crouch_Moving {simSpaceSettings {
     maxAngularAcceleration=1000
     maxAngularVelocity=3
     maxLinearAcceleration=600
     maxLinearVelocity=700
     velocityScaleZ=0
     worldAlpha=0.5
     externalLinearDragV=<0.2 0.2 0>
    }}
   onGround_Falling {simSpaceSettings {
     maxAngularAcceleration=1000
     maxAngularVelocity=3
     maxLinearAcceleration=600
     maxLinearVelocity=700
     velocityScaleZ=0
     worldAlpha=0.5
     externalLinearDragV=<0.2 0.2 0>
    }}
   onGround_Moving {simSpaceSettings {
     maxAngularAcceleration=1000
     maxAngularVelocity=3
     maxLinearAcceleration=600
     maxLinearVelocity=700
     velocityScaleZ=0
     worldAlpha=0.5
     externalLinearDragV=<0.2 0.2 0>
    }}
   onGround_Standing {simSpaceSettings {
     maxAngularAcceleration=1000
     maxAngularVelocity=3
     maxLinearAcceleration=600
     maxLinearVelocity=700
     velocityScaleZ=0
     worldAlpha=0.5
     externalLinearDragV=<0.2 0.2 0>
    }}
   rail_Grinding {simSpaceSettings {
     maxAngularAcceleration=1000
     maxAngularVelocity=3
     maxLinearAcceleration=600
     maxLinearVelocity=700
     velocityScaleZ=0
     worldAlpha=0.5
     externalLinearDragV=<0.2 0.2 0>
    }}
   riding_Moving {simSpaceSettings {
     maxAngularAcceleration=1000
     maxAngularVelocity=3
     maxLinearAcceleration=600
     maxLinearVelocity=700
     velocityScaleZ=0
     worldAlpha=0.5
     externalLinearDragV=<0.2 0.2 0>
    }}
   riding_Moving_Targeting {simSpaceSettings {
     maxAngularAcceleration=1000
     maxAngularVelocity=3
     maxLinearAcceleration=600
     maxLinearVelocity=700
     velocityScaleZ=0
     worldAlpha=0.5
     externalLinearDragV=<0.2 0.2 0>
    }}
   riding_Standing {simSpaceSettings {
     maxAngularAcceleration=1000
     maxAngularVelocity=3
     maxLinearAcceleration=600
     maxLinearVelocity=700
     velocityScaleZ=0
     worldAlpha=0.5
     externalLinearDragV=<0.2 0.2 0>
    }}
   skydive_DivingDown {simSpaceSettings {
     maxAngularAcceleration=1000
     maxAngularVelocity=3
     maxLinearAcceleration=600
     maxLinearVelocity=700
     velocityScaleZ=0
     worldAlpha=0.5
     externalLinearDragV=<0.2 0.2 0>
    }}
   skydive_FloatingDown {
    componentLinearAccScale=<0 0.5 0>
    componentLinearVelScale=<0 0.5 0>
    simSpaceSettings {
     maxAngularAcceleration=1000
     maxAngularVelocity=3
     maxLinearAcceleration=600
     maxLinearVelocity=700
     velocityScaleZ=0
     worldAlpha=0.5
     externalLinearDragV=<0.2 0.2 0>
    }
   }
   skydive_GoingUp {simSpaceSettings {
     maxAngularAcceleration=1000
     maxAngularVelocity=3
     maxLinearAcceleration=600
     maxLinearVelocity=700
     velocityScaleZ=0
     worldAlpha=0.5
     externalLinearDragV=<0.2 0.2 0>
    }}
   skydive_Parachuting {
    componentLinearAccScale=<0 0.1 0.3>
    componentLinearVelScale=<0 0.1 0.3>
    simSpaceSettings {
     maxAngularAcceleration=1000
     maxAngularVelocity=3
     maxLinearAcceleration=600
     maxLinearVelocity=700
     velocityScaleZ=0
     worldAlpha=0.5
     externalLinearDragV=<0.2 0.2 0>
    }
   }
   skydive_Surfing {simSpaceSettings {
     maxAngularAcceleration=1000
     maxAngularVelocity=3
     maxLinearAcceleration=600
     maxLinearVelocity=700
     velocityScaleZ=0
     worldAlpha=0.5
     externalLinearDragV=<0.2 0.2 0>
    }}
   swimming {simSpaceSettings {
     maxAngularAcceleration=1000
     maxAngularVelocity=3
     maxLinearAcceleration=600
     maxLinearVelocity=700
     velocityScaleZ=0
     worldAlpha=0.5
     externalLinearDragV=<0.2 0.2 0>
    }}
   tactical_Sprint {simSpaceSettings {
     maxAngularAcceleration=1000
     maxAngularVelocity=3
     maxLinearAcceleration=600
     maxLinearVelocity=700
     velocityScaleZ=0
     worldAlpha=0.5
     externalLinearDragV=<0.2 0.2 0>
    }}
   ziplining {simSpaceSettings {
     maxAngularAcceleration=1000
     maxAngularVelocity=3
     maxLinearAcceleration=600
     maxLinearVelocity=700
     velocityScaleZ=0
     worldAlpha=0.5
     externalLinearDragV=<0.2 0.2 0>
    }}
  }
 }
 Sim Space Settings {
  type=SimSpaceSettings
  default {
   dampingAlpha=1
   maxAngularAcceleration=10000
   maxAngularVelocity=10000
   maxLinearAcceleration=10000
   maxLinearVelocity=10000
   velocityScaleZ=1
   worldAlpha=0
   externalLinearDragV=<0 0 0>
  }
 }
 ZiplineAlpha {type=number}
]
